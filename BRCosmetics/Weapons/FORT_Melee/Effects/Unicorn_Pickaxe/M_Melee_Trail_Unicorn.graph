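# Additive, unlit, two-sided trail material for the Unicorn pickaxe, built for beam trails and Niagara ribbons.
# Color blends from a mane base into a curve-atlas gradient sampled by stepped V, which SoftenTrail smooths while fading out the eroded triplanar camera-vector sparkle.
# Two panning cloud samples add motion lines, noise modulates brightness along U, and emissive dims at night.
# Opacity combines those clouds, particle alpha, sphere and sine edge masks, and near-camera fade.
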Material /BRCosmetics/Weapons/FORT_Melee/Effects/Unicorn_Pickaxe/M_Melee_Trail_Unicorn
graph {
 properties {
  bUsedWithBeamTrails=true
  bUsedWithNiagaraRibbons=true
  blendMode=BLEND_Additive
  pixelDepthOffsetMode=PDOM_Legacy
  shadingModel=MSM_Unlit
  translucencyLightingMode=TLM_VolumetricPerVertexDirectional
  twoSided=true
  usedShadingModels=Unlit
 }
 nodes [
  n1 = VectorParameter() {parameterName=Color_ManeBase, defaultValue=<rgba 0.6472 0.0022 0.6927 1>}
  n2 = ComponentMask(None=TextureCoordinate()) {g=true}
  n3 = ScalarParameter() {defaultValue=3, parameterName=Steps}
  n4 = Divide(A=Floor(None=Multiply(A=@n2, B=@n3)), B=@n3)
  n5 = CurveAtlasRowParameter(CurveTime=LinearInterpolate(A=Multiply(A=@n4) {constB=2}, B=@n2, Alpha=ScalarParameter() {parameterName=SoftenTrail}) {constB=0}) {
   atlas=&AthenaCurveAtlas01
   curve=/Game/Athena/LinearColorCurves/UnicornPIckaxeCurve
   defaultValue=169
   parameterName=Curve
  }
  n6 = TextureCoordinate() {uTiling=4}
  n7 = Multiply(A=@n6, B=@n6)
  n8 = Saturate(None=ComponentMask(None=Multiply(A=@n7, B=@n7)) {r=true})
  n9 = ParticleColor()
  n10 = CameraVectorWS()
  n11 = TextureObject() {texture=/Game/Weapons/FORT_Melee/Effects/LaserSwordFounders/T_LaserSwordTrailNoise}
  n12 = VectorParameter() {parameterName=TileGlow, defaultValue=<rgba 1 1 1 1>}
  n13 = DynamicParameter() {paramNames [X_Offset,Y_Offset,Z_Offset,Param4]}
  n14 = AppendVector(A=AppendVector(A=@n13.X_Offset, B=@n13.X_Offset), B=@n13.Z_Offset)
  n15 = Multiply(A=MaterialFunctionCall!(
    Dynamic Param=ParticleColor().G
    Alpha=MaterialFunctionCall!(CameraVector=@n10, Texture=@n11, Tiling=@n12.RGB, Offset=@n14, AxisFadeContrast) {materialFunction=/Engine/Functions/Engine_MaterialFunctions01/Texturing/TriplanarCameraVector}.Result
   ) {materialFunction=/Game/Effects/Fort_Effects/Materials/_Functions/Erode_Alpha}.Result) {constB=300}
  n16 = VectorParameter() {parameterName=Motion Lines Tile A, defaultValue=<rgba 0 0.3 0 0>}
  n17 = Panner(Coordinate=Multiply(A=TextureCoordinate(), B=ComponentMask(None=@n16.RGB) {g=true, r=true})) {speedX=-0.2, speedY=-0.2}
  n18 = VectorParameter() {parameterName=Motion Lines Tile B, defaultValue=<rgba 0 0.2 0 0>}
  n19 = Panner(Coordinate=Multiply(A=TextureCoordinate(), B=ComponentMask(None=@n18.RGB) {g=true, r=true})) {speedX=-0.25, speedY=-0.2}
  n20 = ComponentMask(None=Add(
    A=TextureSample(UVs=@n17) {texture=/Game/Effects/Fort_Effects/PlayerHealthDamage/T_DBNO_Cloud}.RGB
    B=TextureSample(UVs=@n19) {texture=&T_DBNO_Cloud}.RGB
   )) {g=true}
  n21 = Add(A=LinearInterpolate(A=ComponentMask(None=@n15) {r=true}, Alpha=ScalarParameter() {parameterName=SoftenTrail}) {constB=0}, B=@n20) {constB=0}
  n22 = TextureCoordinate() {uTiling=0.2}
  n23 = AppendVector(A=ScalarParameter() {defaultValue=-0.65, parameterName=PanSpeedU}, B=Constant())
  n24 = ConstantBiasScale(None=TextureSample(UVs=Panner(Coordinate=@n22, Speed=@n23) {bFractionalPart=true, speedX=-0.65}) {texture=&T_NoisePack_01}.R) {bias=-0.5, scale=0.2}
  n25 = OneMinus(None=ComponentMask(None=TextureCoordinate()) {r=true})
  n26 = Multiply(
   A=Multiply(A=LinearInterpolate(A=@n1.RGB, B=@n5, Alpha=@n8), B=Multiply(A=@n9.R, B=@n21))
   B=LinearInterpolate(Alpha=Add(A=@n24, B=@n25)) {constA=0.01, constB=0.9}
  )
  n27 = ScalarParameter() {defaultValue=0.25, parameterName=NightTimeBrightness}
  MaterialFunctionCall(Emissive=MaterialFunctionCall(Color_Morning=@n26, Color_Day=@n26, Color_Evening=@n26, Color_Night=Multiply(A=@n26, B=@n27)) {materialFunction=&Color_Time_of_Day}.Result) {materialFunction=&MF_FX_EmissiveAdjustments}
  n28 = ParticleColor()
  n29 = ComponentMask(None=TextureCoordinate()) {r=true}
  n30 = Constant() {r=0.5}
  n31 = AppendVector(
   A=@n30
   B=MaterialFunctionCall(Float2=TextureCoordinate()) {materialFunction=&BreakOutFloat2Components}.G
  )
  n32 = ScalarParameter() {defaultValue=0.5, parameterName=Mask Radius}
  n33 = ScalarParameter() {defaultValue=0.9, parameterName=Mask Hardness}
  n34 = OneMinus(None=SphereMask(A=TextureCoordinate() {uTiling=0.5}, B=Constant2Vector() {g=0.82, r=0.85}, Radius=Constant() {r=0.8}, Hardness=Constant()))
  n35 = Multiply(
   A=Saturate(None=Multiply(A=@n29) {constB=3})
   B=Multiply(A=SphereMask(A=@n31, B=@n30, Radius=@n32, Hardness=@n33), B=Power(Base=@n34) {constExponent=5})
  )
  n36 = MaterialFunctionCall(Distance=ScalarParameter() {defaultValue=64, parameterName=NearFadeDistance}) {materialFunction=&Near_Camera_Fade}
  n37 = ComponentMask(None=TextureCoordinate()) {r=true}
  Multiply(A=Multiply(A=Multiply(A=@n28.A, B=@n20), B=Multiply(A=@n35, B=@n36.Result)), B=Multiply(A=Sine(None=@n37) {period=2}) {constB=0.5})
 ]
}
