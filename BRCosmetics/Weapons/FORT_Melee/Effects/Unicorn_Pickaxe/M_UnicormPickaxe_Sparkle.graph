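# Translucent unlit sprite material for Niagara particles, named as the sparkle effect of the Unicorn pickaxe.
# A rounded per-particle random value selects either a spark texture or a starburst texture, both green channel, so particles vary between two shapes.
# The chosen mask is multiplied by particle color RGB into the shared MF_FX_EmissiveAdjustments function, and by particle alpha for transparency.
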Material /BRCosmetics/Weapons/FORT_Melee/Effects/Unicorn_Pickaxe/M_UnicormPickaxe_Sparkle
graph {
 properties {bUsedWithNiagaraSprites=true, blendMode=BLEND_Translucent, pixelDepthOffsetMode=PDOM_Legacy, shadingModel=MSM_Unlit, usedShadingModels=Unlit}
 nodes [
  n1 = LinearInterpolate(
   A=TextureSample() {texture=/Game/CharClasses/Outlander/Gadgets/AntiMaterialCharge/Materials/T_AntiMaterialCharge_Sparks_01}.G
   B=TextureSample() {texture=/Game/Effects/Fort_Effects/Textures/Flares/T_Starburst}.G
   Alpha=Round(None=ParticleRandom())
  )
  n2 = ParticleColor()
  MaterialFunctionCall(Emissive=Multiply(A=@n1, B=@n2.RGB)) {materialFunction=&MF_FX_EmissiveAdjustments}
  Multiply(A=@n1, B=@n2.A)
 ]
}
