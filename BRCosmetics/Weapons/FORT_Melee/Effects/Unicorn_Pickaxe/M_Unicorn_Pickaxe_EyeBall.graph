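# Default Lit material for the eyeball effect on the Unicorn pickaxe cosmetic melee weapon.
# A view-angle term (one minus camera vector dot pixel normal) is squared, raised to Power, then ceiled and clamped into a binary mask that tints EyeColor.
# A DebugTimeSine function driven by SineSpeed also multiplies Power, but that product feeds no other node.
# A Specular scalar is defined; the graph shows no connections to material outputs.
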Material /BRCosmetics/Weapons/FORT_Melee/Effects/Unicorn_Pickaxe/M_Unicorn_Pickaxe_EyeBall
graph {
 properties {pixelDepthOffsetMode=PDOM_Legacy, usedShadingModels=Default Lit}
 nodes [
  n1 = OneMinus(None=DotProduct(A=CameraVectorWS(), B=PixelNormalWS()))
  n2 = ScalarParameter() {defaultValue=13, parameterName=Power}
  n3 = Power(Base=Multiply(A=@n1, B=@n1), Exp=@n2)
  Multiply(A=VectorParameter() {parameterName=EyeColor, defaultValue=<rgba 0.9323 0.6419 0.4111 1>}.RGB, B=Clamp(None=Ceil(None=@n3)))
  Multiply(
   A=@n2
   B=MaterialFunctionCall(Speed=ScalarParameter() {defaultValue=0.2, parameterName=SineSpeed}) {materialFunction=&DebugTimeSine}.Result
  )
  ScalarParameter() {defaultValue=0.25, parameterName=Specular}
 ]
}
